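# MasterReverbSubmixDefault is the engine's default master reverb submix.
# It feeds into MasterSubmixDefault as its parent and has no child submixes.
# Its effect chain holds a single preset, MasterReverbEffectPreset, which processes any audio sent to it.
# The submix mutes when the application is in the background.
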SoundSubmix /Engine/EngineSounds/Submixes/MasterReverbSubmixDefault
graph_type=sound_submix
parent=&/MasterSubmixDefault
children []
properties {bMuteWhenBackgrounded=true, submixEffectChain [&/MasterReverbEffectPreset]}
