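# MasterSubmixDefault is the engine's default master SoundSubmix, sitting at the root of a submix hierarchy with no parent.
# It has two child submixes, MasterEQSubmixDefault and MasterReverbSubmixDefault, whose output mixes into it.
# All of its own properties stay at their default values.
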SoundSubmix /Engine/EngineSounds/Submixes/MasterSubmixDefault
graph_type=sound_submix
parent=null
children [&/MasterEQSubmixDefault, &/MasterReverbSubmixDefault]
properties {}
