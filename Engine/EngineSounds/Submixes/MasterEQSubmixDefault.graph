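# Engine default EQ submix that applies equalization to audio routed through it.
# It is a child of MasterSubmixDefault and has no child submixes.
# Its effect chain holds a single effect preset, MasterEQEffectPreset.
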SoundSubmix /Engine/EngineSounds/Submixes/MasterEQSubmixDefault
graph_type=sound_submix
parent=&/MasterSubmixDefault
children []
properties {submixEffectChain [&/MasterEQEffectPreset]}
